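# Sound cue for small, distant metallic foley movement spots at the StarterIsland bus hanger, routed to the indoor ambient sound class with less-positional one-shot attenuation.
# A random node picks between two equally weighted pools: reused castle metallic rattles or blimp buckle drops, each behind a single-input mixer.
# A quality-level node feeds that selection into only its first two slots, leaving the others empty.
# A root modulator varies pitch and volume.
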SoundCue /Game/Sounds/Helios/POI/StarterIsland/Spots/Spot_StarterIsland_BusHanger_Foley_Movements_Small_Cue
root=Modulator_0
nodes [
 {id=Mixer_1, children [Random_0], properties {inputVolume [1]}}
 {id=Mixer_2, children [Random_5], properties {inputVolume [1]}}
 {id=Modulator_0, children [QualityLevel_0], properties {pitchMin=0.7, volumeMin=0.85}}
 {id=QualityLevel_0, children [Random_8,Random_8,null,null]}
 {
  id=Random_0
  children [WavePlayer_0,WavePlayer_1,WavePlayer_2,WavePlayer_3,WavePlayer_4,WavePlayer_5,WavePlayer_6,WavePlayer_7,WavePlayer_8,WavePlayer_9]
  properties {weights [1,1,1,1,1,1,1,1,1,1]}
 }
 {
  id=Random_5
  children [WavePlayer_34, WavePlayer_35, WavePlayer_36, WavePlayer_37, WavePlayer_38, WavePlayer_39]
  properties {weights [1,1,1,1,1,1]}
 }
 {id=Random_8, children [Mixer_2,Mixer_1], properties {weights [1,1]}}
 {id=WavePlayer_0, properties {soundWaveAssetPtr=/Game/Sounds/Artemis/POI/Blimps/Spots/Spot_Blimp_Metal_Buckle_Drops_Distant_01_Stereo}}
 {id=WavePlayer_1, properties {soundWaveAssetPtr=/Game/Sounds/Artemis/POI/Blimps/Spots/Spot_Blimp_Metal_Buckle_Drops_Distant_03_Stereo}}
 {id=WavePlayer_2, properties {soundWaveAssetPtr=/Game/Sounds/Artemis/POI/Blimps/Spots/Spot_Blimp_Metal_Buckle_Drops_Distant_07_Stereo}}
 {id=WavePlayer_3, properties {soundWaveAssetPtr=/Game/Sounds/Artemis/POI/Blimps/Spots/Spot_Blimp_Metal_Buckle_Drops_Distant_09_Stereo}}
 {id=WavePlayer_34, properties {soundWaveAssetPtr=/Game/Sounds/Asteria/POI/Knights/Spots/Spot_Castle_Metallic_Rattle_Distant_01}}
 {id=WavePlayer_35, properties {soundWaveAssetPtr=/Game/Sounds/Asteria/POI/Knights/Spots/Spot_Castle_Metallic_Rattle_Distant_02}}
 {id=WavePlayer_36, properties {soundWaveAssetPtr=/Game/Sounds/Asteria/POI/Knights/Spots/Spot_Castle_Metallic_Rattle_Distant_03}}
 {id=WavePlayer_37, properties {soundWaveAssetPtr=/Game/Sounds/Asteria/POI/Knights/Spots/Spot_Castle_Metallic_Rattle_Distant_04}}
 {id=WavePlayer_38, properties {soundWaveAssetPtr=/Game/Sounds/Asteria/POI/Knights/Spots/Spot_Castle_Metallic_Rattle_Distant_05}}
 {id=WavePlayer_39, properties {soundWaveAssetPtr=/Game/Sounds/Asteria/POI/Knights/Spots/Spot_Castle_Metallic_Rattle_Distant_06}}
 {id=WavePlayer_4, properties {soundWaveAssetPtr=/Game/Sounds/Artemis/POI/Blimps/Spots/Spot_Blimp_Metal_Buckle_Drops_Distant_10_Stereo}}
 {id=WavePlayer_5, properties {soundWaveAssetPtr=/Game/Sounds/Artemis/POI/Blimps/Spots/Spot_Blimp_Metal_Buckle_Drops_Distant_02_Stereo}}
 {id=WavePlayer_6, properties {soundWaveAssetPtr=/Game/Sounds/Artemis/POI/Blimps/Spots/Spot_Blimp_Metal_Buckle_Drops_Distant_04_Stereo}}
 {id=WavePlayer_7, properties {soundWaveAssetPtr=/Game/Sounds/Artemis/POI/Blimps/Spots/Spot_Blimp_Metal_Buckle_Drops_Distant_05_Stereo}}
 {id=WavePlayer_8, properties {soundWaveAssetPtr=/Game/Sounds/Artemis/POI/Blimps/Spots/Spot_Blimp_Metal_Buckle_Drops_Distant_06_Stereo}}
 {id=WavePlayer_9, properties {soundWaveAssetPtr=/Game/Sounds/Artemis/POI/Blimps/Spots/Spot_Blimp_Metal_Buckle_Drops_Distant_08_Stereo}}
]
properties {
 attenuationSettings=/Game/Sounds/Attenuation/Fort_TileAmbient_OneShot_LessPositional_Log_Att
 maxDistance=6200
 soundClassObject=&Fort_Ambient_Indoors_Athena
 volumeMultiplier=0.501
}
